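# Parent material for the Robuster Wireframe vehicle wheel: a masked, black-based surface with glowing wireframe in paint and accent colors.
# Rim and tire masks, blended by reroute values, are sampled at several BumpOffset depths to fake layered wires.
# Fresnel, screen-space scanlines, stepped-time stripes and optional tread animation feed the color and glow layers.
# Emissive, normal, roughness and vertex-color opacity pass through vehicle damage and de-rez functions.
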Material /VehicleCosmetics/Wheels/Robuster_Wireframe/Materials/M_WheelParent_RobusterWireFrame
graph {
 properties {
  bAutomaticallySetUsageInEditor=false
  bUseMaterialAttributes=true
  bUsedWithSkeletalMesh=true
  blendMode=BLEND_Masked
  usedShadingModels=Default Lit
 }
 nodes [
  n1 = OneMinus(None=VertexColor().G)
  n2 = TextureSampleParameter2D() {parameterName=Opasity_mask, texture=/VehicleCosmetics/Textures/T_White}
  n3 = SetMaterialAttributes!(
   MaterialAttributes
   Base Color=Constant()
   Emissive Color=NamedRerouteUsage()
   Normal=NamedRerouteUsage()
   Opacity=OneMinus(None=LinearInterpolate(A=@n1, Alpha=@n2.RGB) {constB=0})
   Roughness=NamedRerouteUsage()
  ) {attributeSetTypes [
    #1
    #2
    #3
    #4
    #5
   ]}
  MaterialFunctionCall(In=MaterialFunctionCall(Material=@n3) {materialFunction=&MF_BaseVehicle_Damage}.Result) {materialFunction=&MF_VehicleDeRez_Transition}
  n4 = ComponentMask(None=ScreenPosition().ViewportUV) {g=true, r=true}
  n5 = ScalarParameter() {defaultValue=300, parameterName=Scanline Tiling Multiplier, sliderMax=1}
  n6 = TextureSample(UVs=Panner(Coordinate=Multiply(A=@n4, B=@n5)) {speedY=1}) {texture=/VehicleCosmetics/Wheels/Robuster_Wireframe/Textures/LinearGradient_RGB}
  NamedRerouteDeclaration(None=Clamp(None=@n6.R)) {name=ScanLines, nodeColor=<rgba 0.4297 0 1 1>}
  NamedRerouteDeclaration(None=VertexColor().R) {name=VC_RC, nodeColor=<rgba 1 0.0938 0 1>}
  n7 = BumpOffset(Coordinate=NamedRerouteUsage(), Height=Add(A=Constant() {r=0.75}, B=NamedRerouteUsage()), HeightRatioInput=Constant() {r=-0.05})
  n8 = TextureSampleParameter2D(UVs=@n7) {
   parameterName=Rim_wireFrameMAp_Mask
   samplerType=SAMPLERTYPE_LinearColor
   texture=/VehicleCosmetics/Wheels/Robuster_Wireframe/Textures/T_Robusterwireframe_Rim_RGB
  }
  n9 = TextureSampleParameter2D(UVs=@n7) {
   parameterName=Tire_wireFrameMAp_Mask
   samplerSource=SSM_Wrap_WorldGroupSettings
   samplerType=SAMPLERTYPE_LinearColor
   texture=/VehicleCosmetics/Wheels/Robuster_Wireframe/Textures/T_Robusterwireframe_Tire_RGB
  }
  n10 = LinearInterpolate(A=@n8.A, B=@n9.A, Alpha=NamedRerouteUsage())
  n11 = OneMinus(None=@n10)
  n12 = BumpOffset(Coordinate=NamedRerouteUsage(), Height=Add(A=Constant() {r=0.9}, B=NamedRerouteUsage()), HeightRatioInput=Constant() {r=-0.07})
  n13 = TextureSampleParameter2D(UVs=@n12) {
   parameterName=Rim_wireFrameMAp_Mask
   samplerType=SAMPLERTYPE_LinearColor
   texture=&T_Robusterwireframe_Rim_RGB
  }
  n14 = TextureSampleParameter2D(UVs=@n12) {
   parameterName=Tire_wireFrameMAp_Mask
   samplerSource=SSM_Wrap_WorldGroupSettings
   samplerType=SAMPLERTYPE_LinearColor
   texture=&T_Robusterwireframe_Tire_RGB
  }
  n15 = LinearInterpolate(A=@n13.A, B=@n14.A, Alpha=NamedRerouteUsage())
  n16 = Clamp(None=OneMinus(None=@n15))
  n17 = Max(
   A=StaticSwitchParameter(True=Clamp(None=@n11), False=@n10) {parameterName=WireInvert}
   B=Multiply(A=StaticSwitchParameter(True=@n16, False=@n15) {parameterName=WireInvert}, B=Constant() {r=0.6})
  )
  n18 = LinearInterpolate(A=Multiply(A=@n17, B=Constant() {r=0.5}), B=@n17, Alpha=NamedRerouteUsage())
  NamedRerouteDeclaration(None=@n18) {name=AccentLerp, nodeColor=<rgba 0 0.3437 1 1>}
  NamedRerouteDeclaration(None=Multiply(A=TextureCoordinate(), B=ScalarParameter() {defaultValue=1, parameterName=WireUVTiling})) {name=WireTiling, nodeColor=<rgba 0.6641 1 0 1>}
  n19 = BumpOffset(Coordinate=NamedRerouteUsage(), Height=Add(A=Constant() {r=0.65}, B=NamedRerouteUsage()), HeightRatioInput=Constant() {r=-0.05})
  n20 = TextureSampleParameter2D(UVs=@n19) {
   parameterName=Rim_wireFrameMAp_Mask
   samplerType=SAMPLERTYPE_LinearColor
   texture=&T_Robusterwireframe_Rim_RGB
  }
  n21 = TextureSampleParameter2D(UVs=@n19) {
   parameterName=Tire_wireFrameMAp_Mask
   samplerSource=SSM_Wrap_WorldGroupSettings
   samplerType=SAMPLERTYPE_LinearColor
   texture=&T_Robusterwireframe_Tire_RGB
  }
  n22 = LinearInterpolate(A=@n20.A, B=@n21.A, Alpha=NamedRerouteUsage())
  n23 = StaticSwitchParameter(True=Clamp(None=OneMinus(None=@n22)), False=@n22) {parameterName=WireInvert}
  n24 = Multiply(A=@n23, B=Constant() {r=0.5})
  n25 = Multiply(A=@n23, B=Constant() {r=0.9})
  n26 = NamedRerouteUsage()
  n27 = TextureSampleParameter2D(UVs=@n26) {
   parameterName=Rim_wireFrameMAp_Mask
   samplerType=SAMPLERTYPE_LinearColor
   texture=&T_Robusterwireframe_Rim_RGB
  }
  n28 = TextureSampleParameter2D(UVs=@n26) {
   parameterName=Tire_wireFrameMAp_Mask
   samplerSource=SSM_Wrap_WorldGroupSettings
   samplerType=SAMPLERTYPE_LinearColor
   texture=&T_Robusterwireframe_Tire_RGB
  }
  n29 = LinearInterpolate(A=@n27.A, B=@n28.A, Alpha=NamedRerouteUsage())
  n30 = StaticSwitchParameter(True=Clamp(None=OneMinus(None=@n29)), False=@n29) {parameterName=WireInvert}
  n31 = Constant() {r=0.7}
  n32 = Clamp(None=Max(A=LinearInterpolate(A=@n24, B=@n25, Alpha=@n6.R), B=Multiply(A=@n30, B=@n31)))
  n33 = Clamp(None=ConstantBiasScale(None=@n32) {bias=0, scale=3})
  NamedRerouteDeclaration(None=Multiply(A=@n33, B=NamedRerouteUsage())) {name=TeamColor1, nodeColor=<rgba 1 0 0.9844 1>}
  n34 = TextureSampleParameter2D() {
   parameterName=Normal_Rim
   samplerType=SAMPLERTYPE_Normal
   texture=/VehicleCosmetics/Wheels/Robuster_Wireframe/Textures/T_Robusterwireframe_Rim_N
  }
  n35 = TextureSampleParameter2D() {
   parameterName=NormalTire
   samplerType=SAMPLERTYPE_Normal
   texture=/VehicleCosmetics/Wheels/Robuster_Wireframe/Textures/T_Robusterwireframe_Tire_N
  }
  n36 = NamedRerouteUsage()
  n37 = AppendVector(A=LinearInterpolate(A=@n34.R, B=@n35.R, Alpha=@n36), B=LinearInterpolate(A=@n34.G, B=@n35.G, Alpha=@n36))
  n38 = Multiply(A=@n37, B=ScalarParameter() {defaultValue=-1.5, parameterName=Effect Normal})
  n39 = Constant() {r=1}
  n40 = MaterialFunctionCall!(Normal=NamedRerouteDeclaration(None=AppendVector(A=@n38, B=@n39)) {name=NormalInput, nodeColor=<rgba 0 0.2266 1 1>}, Exponent) {materialFunction=/VehicleCosmetics/Wheels/Cola_Holographic/Materials/MF_Fresnel_Classic}
  n41 = Power(Base=@n40.Result, Exp=Constant() {r=1})
  n42 = Power(Base=@n40.Result, Exp=Constant() {r=30})
  n43 = NamedRerouteUsage()
  n44 = OneMinus(None=@n40.Result)
  n45 = Clamp(None=Power(Base=@n44, Exp=Constant() {r=20}))
  n46 = NamedRerouteUsage()
  n47 = Max(A=Clamp(None=LinearInterpolate(A=@n41, B=@n42, Alpha=@n43)), B=ConstantBiasScale(None=Multiply(A=@n45, B=@n46)) {bias=0, scale=0.9})
  n48 = Add(A=@n47, B=@n32)
  NamedRerouteDeclaration(None=Clamp(None=@n48)) {name=TeamColor1_2, nodeColor=<rgba 1 0 0.9844 1>}
  NamedRerouteDeclaration(None=@n33) {name=TeamColor1_01, nodeColor=<rgba 1 0.375 0 1>}
  NamedRerouteDeclaration(None=@n48) {name=TeamColorAdd, nodeColor=<rgba 1 0.4453 0 1>}
  NamedRerouteDeclaration(None=@n32) {name=TeamColor_Clamp, nodeColor=<rgba 1 0.4453 0 1>}
  n49 = Constant() {r=2}
  n50 = Clamp(None=Multiply(A=Power(Base=@n18, Exp=@n49), B=Constant() {r=3}))
  n51 = NamedRerouteUsage()
  NamedRerouteDeclaration(None=Clamp(None=Subtract(A=@n50, B=@n51))) {name=Accent_Clamp, nodeColor=<rgba 0.1953 1 0 1>}
  n52 = NamedRerouteUsage()
  n53 = LinearInterpolate(A=@n52, B=@n52, Alpha=NamedRerouteUsage())
  n54 = Constant() {r=0.75}
  n55 = Multiply(
   A=StaticSwitchParameter(True=LinearInterpolate(A=@n52, B=@n53, Alpha=@n54), False=@n52) {defaultValue=true, parameterName=Painted}
   B=NamedRerouteUsage()
  )
  n56 = Constant() {r=0.4}
  n57 = Constant()
  n58 = Constant()
  n59 = Desaturation(None=Multiply(
    A=MaterialFunctionCall!(Tint=@n57, TintAmount=@n57, Desaturate=@n58) {materialFunction=/VehicleCosmetics/Wheels/Oldmans/New/Materials/MF_WordENV}.LocalSpace
    B=Constant() {r=0.5}
   )) {luminanceFactors=<rgba 0.3 0.59 0.11 0>}
  n60 = Constant()
  n61 = Max(A=NamedRerouteUsage(), B=NamedRerouteUsage())
  NamedRerouteDeclaration(None=Multiply(A=Desaturation(None=@n55, Fraction=@n56) {luminanceFactors=<rgba 0.3 0.59 0.11 0>}, B=LinearInterpolate(A=@n59, B=@n60, Alpha=@n61))) {name=ENV, nodeColor=<rgba 1 0 0.5156 1>}
  n62 = VectorParameter() {parameterName=PaintColor, defaultValue=<rgba 0.0203 0.3968 0.6376 1>}
  n63 = StaticSwitchParameter(
   True=LinearInterpolate(A=@n62.RGB, B=VectorParameter() {parameterName=SecondaryColor, defaultValue=<rgba 0.0203 0.3968 0.6376 1>}.RGB, Alpha=NamedRerouteUsage())
   False=@n62.RGB
  ) {parameterName=SecondaryColor}
  n64 = ConstantBiasScale(None=@n63) {bias=0, scale=1}
  n65 = Power(Base=Clamp(None=@n44), Exp=Constant() {r=10})
  n66 = Multiply(A=LinearInterpolate(A=@n64, B=@n63, Alpha=Clamp(None=@n65)), B=@n47)
  n67 = NamedRerouteUsage()
  n68 = NamedRerouteUsage()
  n69 = NamedRerouteUsage()
  n70 = LinearInterpolate(
   A=LinearInterpolate(A=LinearInterpolate(A=@n66, B=@n67, Alpha=@n68), B=NamedRerouteUsage(), Alpha=@n69)
   B=NamedRerouteUsage()
   Alpha=NamedRerouteUsage()
  )
  n71 = Max(A=ConstantBiasScale(None=@n47) {bias=0.1, scale=1}, B=@n69)
  n72 = ScalarParameter() {defaultValue=2, parameterName=HighlightMultiplier}
  n73 = LinearInterpolate(A=@n70, B=Multiply(A=Multiply(A=@n71, B=@n72), B=NamedRerouteUsage()), Alpha=ConstantBiasScale(None=NamedRerouteUsage()) {bias=0.2, scale=1})
  n74 = Add(A=Clamp(None=StaticSwitchParameter(True=@n73, False=@n70) {defaultValue=true, parameterName=Painted}), B=NamedRerouteUsage())
  n75 = ScalarParameter() {defaultValue=2, parameterName=EmissiveIntensity}
  n76 = OneMinus(None=TextureSampleParameter2D() {parameterName=Mask, texture=&T_White}.RGB)
  n77 = VertexColor()
  NamedRerouteDeclaration(None=Multiply(A=Multiply(A=@n74, B=@n75), B=LinearInterpolate(A=@n76, Alpha=@n77.R))) {name=Emissive, nodeColor=<rgba 0 0.4375 1 1>}
  NamedRerouteDeclaration(None=@n37) {name=NormalAppend, nodeColor=<rgba 0.7578 1 0 1>}
  n78 = NamedRerouteUsage()
  n79 = ScalarParameter() {defaultValue=1, parameterName=Surface Normal}
  NamedRerouteDeclaration(None=AppendVector(A=Multiply(A=@n78, B=@n79), B=Constant() {r=1})) {name=Normal, nodeColor=<rgba 0.9219 0 1 1>}
  NamedRerouteDeclaration(None=LinearInterpolate(
    A=ScalarParameter() {defaultValue=0.4, parameterName=Roughness}
    Alpha=TextureSampleParameter2D() {parameterName=Mask, texture=&T_White}.RGB
   )) {name=Rough, nodeColor=<rgba 0 1 0.9062 1>}
  n80 = TextureCoordinate()
  n81 = TextureSampleParameter2D(UVs=@n80) {
   parameterName=Tire_wireFrameMAp_Mask
   samplerSource=SSM_Wrap_WorldGroupSettings
   samplerType=SAMPLERTYPE_LinearColor
   texture=&T_Robusterwireframe_Tire_RGB
  }
  n82 = Multiply(A=Floor(None=Time()), B=ScalarParameter() {defaultValue=1, parameterName=greenpan speed})
  n83 = TextureSampleParameter2D(UVs=Panner(Coordinate=@n80, Time=@n82) {speedX=0.1}) {
   parameterName=Tire_wireFrameMAp_Mask
   samplerSource=SSM_Wrap_WorldGroupSettings
   samplerType=SAMPLERTYPE_LinearColor
   texture=&T_Robusterwireframe_Tire_RGB
  }
  n84 = Multiply(A=Clamp(None=ConstantBiasScale(None=@n81.R) {bias=0, scale=10}), B=@n83.G)
  n85 = NamedRerouteUsage()
  n86 = TextureSampleParameter2D(UVs=@n80) {
   parameterName=Rim_wireFrameMAp_Mask
   samplerType=SAMPLERTYPE_LinearColor
   texture=&T_Robusterwireframe_Rim_RGB
  }
  NamedRerouteDeclaration(None=LinearInterpolate(A=@n86.G, B=StaticSwitchParameter(True=@n84, False=@n81.G) {parameterName=TreadGreenPan}, Alpha=@n85)) {name=Lerp1, nodeColor=<rgba 0.1016 1 0 1>}
  n87 = TextureSampleParameter2D(UVs=Rotator(Coordinate=@n80, Time=@n82) {centerX=0.2743, centerY=0.6951}) {
   parameterName=Rim_wireFrameMAp_Mask
   samplerType=SAMPLERTYPE_LinearColor
   texture=&T_Robusterwireframe_Rim_RGB
  }
  n88 = Multiply(A=@n87.R, B=Clamp(None=ConstantBiasScale(None=@n86.G) {bias=0, scale=10}))
  NamedRerouteDeclaration(None=LinearInterpolate(A=StaticSwitchParameter(True=@n88, False=@n86.R) {parameterName=TreadGreenPan}, B=@n81.R, Alpha=@n85)) {name=Lerp2, nodeColor=<rgba 1 0 0.4219 1>}
  n89 = Time()
  n90 = ScalarParameter() {defaultValue=0.5, parameterName=Time Multiplier}
  n91 = Abs(None=Sine(None=Multiply(A=@n89, B=@n90)) {period=3})
  n92 = Constant() {r=60}
  n93 = Divide(A=Floor(None=Multiply(A=@n91, B=@n92)), B=Constant() {r=30})
  n94 = Subtract(A=NamedRerouteUsage(), B=Clamp(None=OneMinus(None=@n93)))
  n95 = Time()
  n96 = Sine(None=Subtract(A=Multiply(A=@n95, B=@n90), B=Constant() {r=0.25})) {period=3}
  n97 = Constant() {r=60}
  n98 = Divide(A=Floor(None=Multiply(A=@n96, B=@n97)), B=Constant() {r=30})
  n99 = Subtract(A=NamedRerouteUsage(), B=Clamp(None=OneMinus(None=@n98)))
  n100 = Clamp(None=Max(A=Clamp(None=@n94) {maxDefault=0.75}, B=Clamp(None=@n99) {maxDefault=0.75}))
  NamedRerouteDeclaration(None=Clamp(None=ConstantBiasScale(None=@n100) {bias=0, scale=5})) {name=PaimtedStripes, nodeColor=<rgba 0 0.7422 1 1>}
  n101 = Cosine(None=Multiply(A=Constant() {r=0.25}, B=Time()))
  NamedRerouteDeclaration(None=Multiply(A=Abs(None=@n101), B=Constant() {r=0.05})) {name=Timer, nodeColor=<rgba 1 0.9375 0 1>}
  NamedRerouteDeclaration(None=@n64) {name=SecondaryColor, nodeColor=<rgba 0.6172 0 1 1>}
  NamedRerouteDeclaration(None=@n63) {name=SecondaryColorSwitch, nodeColor=<rgba 0.6172 0 1 1>}
  n102 = NamedRerouteUsage()
  n103 = Constant() {r=0.02}
  NamedRerouteDeclaration(None=LinearInterpolate(A=Multiply(A=@n102, B=@n103), B=NamedRerouteUsage(), Alpha=@n50)) {name=Accent, nodeColor=<rgba 1 0 0.0234 1>}
 ]
}
